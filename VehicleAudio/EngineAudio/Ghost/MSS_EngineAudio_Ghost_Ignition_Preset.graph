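# MetaSound Source preset that plays the Ghost vehicle's engine ignition as a mono one-shot.
# It inherits its graph from MSS_EngineAudio_Ignition, supplying Ghost ignition waves and borrowed Lamborghini Huracan exhaust pops.
# Inputs set the pops' start delay, count range, repeat rate range, pitch and volume.
# Shared vehicle attenuation, ignition concurrency and engine modulators for volume, highpass and lowpass shape the output.
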MetaSoundSource /VehicleAudio/EngineAudio/Ghost/MSS_EngineAudio_Ghost_Ignition_Preset
preset_parent=/VehicleAudio/Patches/MSS_EngineAudio_Ignition
inputs [
 ExhaustPop_StartDelayTime {type=Time, access=Constructor}
 ExhaustPopPitch {type=Float}
 ExhaustPops {type=WaveAsset:Array, access=Constructor}
 ExhaustPopVolume {type=Float, access=Constructor}
 IgnitionSound {type=WaveAsset:Array, access=Constructor}
 IgnitionVolume {type=Float, access=Constructor}
 MaxNumExhaustPopsToPlay {type=Int32, access=Constructor}
 MaxRepeatRate {type=Float, access=Constructor}
 MinNumExhaustPopsToPlay {type=Int32, access=Constructor}
 MinRepeatRate {type=Float, access=Constructor}
 UE.Source.OnPlay {type=Trigger}
]
interfaces [UE.OutputFormat.Mono:1.0, UE.Source.OneShot:1.0, UE.Source:1.1]
outputs [
 "UE.OutputFormat.Mono.Audio:0" {type=Audio}
 UE.Source.OneShot.OnFinished {type=Trigger}
]
pages [Default {input_defaults {
   ExhaustPopPitch=-4.5
   ExhaustPopVolume=-8
   ExhaustPop_StartDelayTime=0.4
   IgnitionVolume=0
   MaxNumExhaustPopsToPlay=14
   MaxRepeatRate=0.2
   MinNumExhaustPopsToPlay=8
   MinRepeatRate=0.1
   UE.Source.OnPlay=null
   ExhaustPops [
    /VehicleAudio/EngineAudio/Lambo/Pops/EngineAudio_LamboHuracan_ExhaustPop_Long_01
    /VehicleAudio/EngineAudio/Lambo/Pops/EngineAudio_LamboHuracan_ExhaustPop_Long_02
    /VehicleAudio/EngineAudio/Lambo/Pops/EngineAudio_LamboHuracan_ExhaustPop_Short_01
    /VehicleAudio/EngineAudio/Lambo/Pops/EngineAudio_LamboHuracan_ExhaustPop_Short_02
    /VehicleAudio/EngineAudio/Lambo/Pops/EngineAudio_LamboHuracan_ExhaustPop_Short_03
    /VehicleAudio/EngineAudio/Lambo/Pops/EngineAudio_LamboHuracan_ExhaustPop_Short_04
    /VehicleAudio/EngineAudio/Lambo/Pops/EngineAudio_LamboHuracan_ExhaustPop_Short_05
    /VehicleAudio/EngineAudio/Lambo/Pops/EngineAudio_LamboHuracan_ExhaustPop_Short_06
    /VehicleAudio/EngineAudio/Lambo/Pops/EngineAudio_LamboHuracan_ExhaustPop_Short_07
    /VehicleAudio/EngineAudio/Lambo/Pops/EngineAudio_LamboHuracan_ExhaustPop_Short_08
    /VehicleAudio/EngineAudio/Lambo/Pops/EngineAudio_LamboHuracan_ExhaustPop_Short_09
    /VehicleAudio/EngineAudio/Lambo/Pops/EngineAudio_LamboHuracan_ExhaustPop_Short_10
    /VehicleAudio/EngineAudio/Lambo/Pops/EngineAudio_LamboHuracan_ExhaustPop_Short_11
    /VehicleAudio/EngineAudio/Lambo/Pops/EngineAudio_LamboHuracan_ExhaustPop_Short_12
    /VehicleAudio/EngineAudio/Lambo/Pops/EngineAudio_LamboHuracan_ExhaustPop_Short_13
    /VehicleAudio/EngineAudio/Lambo/Pops/EngineAudio_LamboHuracan_ExhaustPop_Short_14
   ]
   IgnitionSound [
    &/EngineAudio_Ghost_Ignition_Mono_03
    &/EngineAudio_Ghost_Ignition_Mono_02
    &/EngineAudio_Ghost_Ignition_Mono_01
   ]
  }}]
properties {
 assetImportData=&self:AssetImportData
 attenuationSettings=/VehicleAudio/Settings/Vehicle_Movement_Att
 audioPropertiesSheet=/VehicleAudio/PropertySheets/APS_Cosmetic_Vehicle_Engine_Ignition
 maxDistance=12512
 sampleRate=48000
 soundClassObject=&Fort_Vehicles_Interior
 soundSubmixObject=&Fortnite_SFX
 volume=1.413
 concurrencySet [/VehicleAudio/Settings/Vehicle_Cosmetic_Ignition_Conc]
 modulationSettings {
  highpassRouting=Union
  lowpassRouting=Union
  volumeRouting=Union
  highpassModulationDestination {bEnableModulation=true, modulators [&PP_FortVehicles_Engine_HPF]}
  lowpassModulationDestination {bEnableModulation=true, modulators [&Patch_FortVehicles_Engine_LPF]}
  volumeModulationDestination {
   bEnableModulation=true
   modulators [&Patch_FortVehicles_Engine_Vol, &DM_VehicleEnginePatch]
  }
 }
}
